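# Physics asset for the Weeping Woods Spring male one-handed axe pickaxe cosmetic, giving its six flower bones physical secondary motion.
# A small capsule on the handle bone serves as the shared anchor for six identical capsules on the dyn_flower bones.
# Each flower is constrained directly to the handle rather than chained, with moderate swing and tight twist limits so it can sway without spinning.
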PhysicsAsset /BRCosmetics/Weapons/FORT_Melee/Pickaxe_WeepingWoodsSpringMale1H/Meshes/Weeping_Woods_Spring_Male_1h_Axe_Dyn
bodies {
 Melee_TwoHanded_Handle [{shapeType=Capsule, length=0.5, radius=0.5}]
 dyn_flower_01 [{shapeType=Capsule, length=2.7859, radius=0.8745, center=<0.1 0 0>, rotation=<rot 90 0 0>}]
 dyn_flower_02 [{shapeType=Capsule, length=2.7859, radius=0.8745, center=<0.1 0 0>, rotation=<rot 90 0 0>}]
 dyn_flower_03 [{shapeType=Capsule, length=2.7859, radius=0.8745, center=<0.1 0 0>, rotation=<rot 90 0 0>}]
 dyn_flower_04 [{shapeType=Capsule, length=2.7859, radius=0.8745, center=<0.1 0 0>, rotation=<rot 90 0 0>}]
 dyn_flower_05 [{shapeType=Capsule, length=2.7859, radius=0.8745, center=<0.1 0 0>, rotation=<rot 90 0 0>}]
 dyn_flower_06 [{shapeType=Capsule, length=2.7859, radius=0.8745, center=<0.1 0 0>, rotation=<rot 90 0 0>}]
}
constraints [
 {bone1Name=dyn_flower_01, bone2Name=Melee_TwoHanded_Handle, swing1LimitDegrees=45, swing2LimitDegrees=45, twistLimitDegrees=5}
 {bone1Name=dyn_flower_02, bone2Name=Melee_TwoHanded_Handle, swing1LimitDegrees=45, swing2LimitDegrees=45, twistLimitDegrees=5}
 {bone1Name=dyn_flower_03, bone2Name=Melee_TwoHanded_Handle, swing1LimitDegrees=45, swing2LimitDegrees=45, twistLimitDegrees=5}
 {bone1Name=dyn_flower_04, bone2Name=Melee_TwoHanded_Handle, swing1LimitDegrees=45, swing2LimitDegrees=45, twistLimitDegrees=5}
 {bone1Name=dyn_flower_05, bone2Name=Melee_TwoHanded_Handle, swing1LimitDegrees=45, swing2LimitDegrees=45, twistLimitDegrees=5}
 {bone1Name=dyn_flower_06, bone2Name=Melee_TwoHanded_Handle, swing1LimitDegrees=45, swing2LimitDegrees=45, twistLimitDegrees=5}
]
